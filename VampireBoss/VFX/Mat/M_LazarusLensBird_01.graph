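# Masked two-sided material for Lazarus Lens bird mesh particles, apparently for a skydiving contrail effect.
# Several vertex-offset networks build sine-driven body waves, bobbing and wing flaps, using vertex-colour masks and rotation about object or instance-space pivots.
# Dynamic particle parameters set time bias, speed and X scale per particle.
# Shading combines colour, normal and mask textures with particle colour, a UV1-blended two-colour tint and emissive adjustment.
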Material /VampireBoss/VFX/Mat/M_LazarusLensBird_01
graph {
 properties {
  bFullyRough=true
  bTangentSpaceNormal=false
  bUsedWithMeshParticles=true
  bUsedWithNiagaraMeshParticles=true
  bUsedWithStaticLighting=true
  blendMode=BLEND_Masked
  numCustomizedUVs=2
  pixelDepthOffsetMode=PDOM_Legacy
  previewMesh=/VampireBoss/VFX/SM_LazarusLens_Bird_NPC1
  twoSided=true
  usedShadingModels=Default Lit
 }
 nodes [
  n1 = Transform(None=Constant3Vector() {constant=<rgba 0 0 1 0>}) {transformSourceType=TRANSFORMSOURCE_Local}
  n2 = DynamicParameter() {defaultValue=<rgba 0 0 1 1>, paramNames [Bias, Speed, particle X Scale, Param4]}
  n3 = Multiply(A=Add(A=Time(), B=@n2.Bias), B=@n2.Speed)
  n4 = ScalarParameter() {parameterName=Macro Time Bias}
  n5 = Transform(None=WorldPosition().XYZ) {transformSourceType=TRANSFORMSOURCE_World, transformType=TRANSFORM_Local}
  n6 = ScalarParameter() {defaultValue=128, parameterName=Macro Wave Frequency}
  n7 = ScalarParameter() {defaultValue=128, parameterName=Macro Wave Frequency Wings}
  n8 = Multiply(A=VertexColor().A, B=ScalarParameter() {parameterName=UseMacroWingFreq?})
  n9 = Add(
   A=Multiply(A=Add(A=@n3, B=@n4), B=ScalarParameter() {defaultValue=1, parameterName=Speed})
   B=Divide(A=ComponentMask(None=@n5) {r=true}, B=LinearInterpolate(A=@n6, B=@n7, Alpha=@n8))
  )
  n10 = Multiply(A=Multiply(A=Sine(None=@n9), B=ScalarParameter() {defaultValue=1, parameterName=Macro Wave scale}), B=VertexColor().B)
  n11 = Constant3Vector() {constant=<rgba 1 0 0 0>}
  n12 = Subtract(A=Multiply(A=VertexColor().R) {constB=2})
  n13 = Multiply(A=@n3, B=ScalarParameter() {defaultValue=1, parameterName=Speed})
  n14 = Add(A=Multiply(A=Sine(None=@n13), B=ScalarParameter() {parameterName=Rotate Amount}), B=ScalarParameter() {parameterName=Rotation Angle Bias})
  n15 = Constant3Vector() {constant=<rgba 0 1 0 0>}
  n16 = Subtract(A=Multiply(A=VertexColor().G) {constB=2})
  n17 = Add(A=ScalarParameter() {parameterName=Forward Rotate Delay}, B=@n3)
  n18 = ScalarParameter() {defaultValue=1, parameterName=Speed}
  n19 = Multiply(A=Sine(None=Multiply(A=@n17, B=@n18)), B=ScalarParameter() {parameterName=Forward rotate Amount})
  n20 = Add(
   A=RotateAboutAxis(
    NormalizedRotationAxis=Transform(None=@n11) {transformSourceType=TRANSFORMSOURCE_Local}
    RotationAngle=Multiply(A=@n12, B=@n14)
    PivotPoint=ObjectPositionWS()
    Position=WorldPosition().XYZ
   )
   B=RotateAboutAxis(
    NormalizedRotationAxis=Transform(None=@n15) {transformSourceType=TRANSFORMSOURCE_Local}
    RotationAngle=Multiply(A=@n16, B=@n19)
    PivotPoint=ObjectPositionWS()
    Position=WorldPosition().XYZ
   )
  )
  n21 = VertexColor()
  n22 = DynamicParameter() {defaultValue=<rgba 0 0 1 1>, paramNames [Bias, Speed, particle X Scale, Param4]}
  n23 = Sine(None=Multiply(A=@n22.Param4, B=Time())) {period=3}
  n24 = Constant3Vector() {constant=<rgba 0 0 1 0>}
  Add(A=Add(A=Multiply(A=@n1, B=@n10), B=Multiply(A=@n20, B=@n21.A)), B=Multiply(A=@n22.particle X Scale, B=Multiply(A=@n23, B=@n24)))
  n25 = Constant3Vector() {constant=<rgba 1 0 0.3 0>}
  n26 = DynamicParameter() {defaultValue=<rgba 0 0 1 1>, paramNames [Bias, Speed, particle X Scale, Param4]}
  n27 = Multiply(A=Add(A=Time(), B=@n26.Bias), B=@n26.Speed)
  n28 = Multiply(A=OneMinus(None=VertexColor().A)) {constB=-0.05}
  n29 = ConstantBiasScale(None=Sine(None=Add(A=@n27, B=@n28)))
  n30 = LinearInterpolate(Alpha=@n29) {constA=0.235, constB=-0.28}
  n31 = OneMinus(None=@n30)
  n32 = VertexColor()
  n33 = Saturate(None=Add(A=@n32.R, B=@n32.R))
  n34 = Transform(None=Constant3Vector() {constant=<rgba 0 4 0 0>}) {transformSourceType=TRANSFORMSOURCE_Instance}
  n35 = Add(
   A=TransformPosition(None=Constant3Vector()) {transformSourceType=TRANSFORMPOSSOURCE_Instance, transformType=TRANSFORMPOSSOURCE_World}
   B=LinearInterpolate(A=@n34, B=Multiply(A=@n34) {constB=-1}, Alpha=VertexColor().R)
  )
  n36 = TransformPosition(None=WorldPosition().XYZ) {transformSourceType=TRANSFORMPOSSOURCE_Instance, transformType=TRANSFORMPOSSOURCE_World}
  n37 = RotateAboutAxis(
   NormalizedRotationAxis=Transform(None=Normalize(VectorInput=@n25)) {transformSourceType=TRANSFORMSOURCE_Instance}
   RotationAngle=LinearInterpolate(A=LinearInterpolate(A=@n30, B=@n31, Alpha=@n32.R) {constAlpha=0.75}, Alpha=OneMinus(None=@n33))
   PivotPoint=@n35
   Position=@n36
  )
  n38 = Normalize(VectorInput=Constant3Vector() {constant=<rgba 0 1 0 0>})
  n39 = LinearInterpolate(Alpha=@n29) {constB=-0.08}
  n40 = VertexColor()
  n41 = OneMinus(None=Saturate(None=Add(A=@n40.R, B=@n40.R)))
  Add(
   A=@n37
   B=RotateAboutAxis(
    NormalizedRotationAxis=Transform(None=@n38) {transformSourceType=TRANSFORMSOURCE_Instance}
    RotationAngle=LinearInterpolate(A=@n39, Alpha=@n41)
    PivotPoint=@n35
    Position=Add(A=@n37, B=@n36)
   )
  )
  Add(
   A=TextureSample() {texture=/VampireBoss/Athena/Effects/Particles/Skydiving/NS_Contrail_LazarusLens/Textures/T_LazarusLense_Contrail_FX}.R
   B=ScalarParameter() {parameterName=Opac Mask Offset}
  )
  n42 = TransformPosition(None=Constant3Vector()) {transformSourceType=TRANSFORMPOSSOURCE_Instance, transformType=TRANSFORMPOSSOURCE_World}
  n43 = WorldPosition()
  n44 = Constant3Vector() {constant=<rgba 1 0 0.3 0>}
  n45 = Multiply(A=OneMinus(None=VertexColor().A)) {constB=-0.05}
  n46 = DynamicParameter() {defaultValue=<rgba 0 0 1 1>, paramNames [Bias, Speed, particle X Scale, Param4]}
  n47 = Multiply(A=Add(A=Time(), B=@n46.Bias), B=@n46.Speed)
  n48 = ConstantBiasScale(None=Sine(None=Add(A=@n47, B=@n45)))
  n49 = LinearInterpolate(Alpha=@n48) {constA=0.235, constB=-0.28}
  n50 = OneMinus(None=@n49)
  n51 = VertexColor()
  n52 = Saturate(None=Add(A=@n51.R, B=@n51.R))
  n53 = Transform(None=Multiply(
    A=Constant3Vector() {constant=<rgba 0 4 0 0>}
    B=DynamicParameter() {defaultValue=<rgba 0 0 1 1>, paramNames [Bias, Speed, particle X Scale, Param4]}.particle X Scale
   )) {transformSourceType=TRANSFORMSOURCE_Instance}
  n54 = Add(A=@n42, B=LinearInterpolate(A=@n53, B=Multiply(A=@n53) {constB=-1}, Alpha=VertexColor().R))
  n55 = WorldPosition()
  n56 = RotateAboutAxis(
   NormalizedRotationAxis=Transform(None=Normalize(VectorInput=@n44)) {transformSourceType=TRANSFORMSOURCE_Instance}
   RotationAngle=LinearInterpolate(A=LinearInterpolate(A=@n49, B=@n50, Alpha=@n51.R) {constAlpha=0.75}, Alpha=OneMinus(None=@n52))
   PivotPoint=@n54
   Position=@n55.XYZ
  )
  n57 = Constant3Vector() {constant=<rgba 0 1 0 0>}
  n58 = LinearInterpolate(Alpha=@n48) {constB=-0.08}
  n59 = VertexColor()
  n60 = Saturate(None=Add(A=@n59.R, B=@n59.R))
  n61 = RotateAboutAxis(
   NormalizedRotationAxis=Transform(None=Normalize(VectorInput=@n57)) {transformSourceType=TRANSFORMSOURCE_Instance}
   RotationAngle=LinearInterpolate(A=@n58, Alpha=OneMinus(None=@n60))
   PivotPoint=@n54
   Position=Add(A=@n56, B=@n55.XYZ)
  )
  n62 = Constant3Vector() {constant=<rgba 0 0 5 0>}
  n63 = Add(A=@n47, B=LinearInterpolate(Alpha=VertexColor().A) {constA=-0.25, constB=-0.15}) {constB=-0.25}
  Add(
   A=LinearInterpolate(
    A=Subtract(A=@n42, B=@n43.XYZ)
    B=Add(A=@n56, B=@n61)
    Alpha=DynamicParameter() {defaultValue=<rgba 0 0 1 1>, paramNames [Bias, Speed, particle X Scale, Param4]}.particle X Scale
   )
   B=Multiply(A=Transform(None=@n62) {transformSourceType=TRANSFORMSOURCE_Instance}, B=Sine(None=@n63))
  )
  n64 = CollectionParameter() {
   collection=&FortniteMaterialParameters
   parameterName=SunlightDirectionalVector
  }
  Clamp(None=DotProduct(A=ComponentMask(None=@n64) {b=true, g=true, r=true}, B=VertexNormalWS()))
  ComponentMask(None=ParticleSize()) {r=true}
  Constant()
  n65 = Constant() {r=1}
  If(A=@n2.Speed, B=Constant(), "A > B"=@n2.Speed, "A == B"=@n65, "A < B"=@n65)
  n66 = TextureSampleParameter2D() {
   parameterName=ColorTexture
   texture=/VampireBoss/Athena/Effects/Particles/Skydiving/NS_Contrail_LazarusLens/Textures/T_LazarusLense_Contrail_D
  }
  LinearInterpolate(
   A=ScalarParameter() {defaultValue=1, parameterName=RoughnessHigh}
   B=ScalarParameter() {defaultValue=0.5, parameterName=RoughnessLOW}
   Alpha=@n66.R
  )
  LinearInterpolate(A=@n39, B=OneMinus(None=@n39), Alpha=@n40.R) {constAlpha=0.75}
  LinearInterpolate(A=@n58, B=OneMinus(None=@n58), Alpha=@n59.R) {constAlpha=0.75}
  n67 = ParticleColor()
  n68 = LinearInterpolate(
   A=VectorParameter() {parameterName=Color2, defaultValue=<rgba 0.3 0.3 0.3 1>}.RGB
   B=VectorParameter() {parameterName=Color1, defaultValue=<rgba 0.515 0.515 0.515 1>}.RGB
   Alpha=ComponentMask(None=TextureCoordinate() {coordinateIndex=1}) {r=true}
  )
  n69 = Multiply(A=@n66.RGB, B=ParticleColor().RGB)
  n70 = Multiply(A=@n68, B=@n69)
  MaterialFunctionCall(Emissive=Add(A=ScalarParameter() {parameterName=Emissive}, B=Multiply(A=@n67.RGB, B=@n70))) {materialFunction=&MF_FX_EmissiveAdjustments}
  MaterialFunctionCall() {materialFunction=&DebugTimeSine}
  Multiply(A=@n68, B=@n69)
  TextureSampleParameter2D() {
   parameterName=NormalTexture
   samplerType=SAMPLERTYPE_Normal
   texture=/VampireBoss/Athena/Effects/Particles/Skydiving/NS_Contrail_LazarusLens/Textures/T_LazarusLense_Contrail_N
  }
  VectorParameter() {parameterName=Color, defaultValue=<rgba 1 1 1 1>}
 ]
}
