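# Translucent, two-sided unlit material for a Niagara ribbon, drawing the vampire boss's life-drain energy beam.
# Noise-distorted, scrolling smoke and trail textures form the body, with three offset trail samples splitting RGB for chromatic fringing and a bright Center-colored core.
# Curve atlas color along U and particle color tint emissive.
# Opacity uses Dissolve-driven noisy U erosion, end fading and path-switched depth fade, which also feeds a quality-switched refraction value.
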Material /VampireBoss/VFX/Mat/M_LifeDrain_Beam_EnergyVampire
graph {
 properties {
  bUseTranslucencyVertexFog=false
  bUsedWithNiagaraRibbons=true
  blendMode=BLEND_Translucent
  pixelDepthOffsetMode=PDOM_Legacy
  refractionMethod=RM_IndexOfRefraction
  shadingModel=MSM_Unlit
  twoSided=true
  usedShadingModels=Unlit
 }
 nodes [
  n1 = ComponentMask(None=TextureCoordinate()) {r=true}
  n2 = TextureCoordinate() {uTiling=2}
  n3 = Multiply(A=Constant2Vector() {r=-1}, B=Time())
  n4 = ComponentMask(None=TextureSample(UVs=Add(A=@n2, B=@n3)) {samplerType=SAMPLERTYPE_Grayscale, texture=&LowResBlurredNoise}.RGB) {g=true, r=true}
  n5 = Add(A=Multiply(A=@n4) {constB=0.2}) {constB=-0.1}
  n6 = Constant2Vector() {r=-1}
  n7 = Time()
  n8 = Add(A=@n5, B=Add(A=TextureCoordinate(), B=Multiply(A=@n6, B=@n7)))
  n9 = MaterialFunctionCall!(TexCoord=@n8, ClampChannel=Constant2Vector() {g=1}) {materialFunction=/Game/Effects/Fort_Effects/MaterialFunctions/MF_FX_TilingMethod}
  n10 = TextureSample!(UVs=@n8, Tex, Level=Constant() {r=5}, Apply View MipBias) {mipValueMode=TMVM_MipLevel, texture=/Game/Effects/Fort_Effects/Textures/Smoke/T_Smoke_Trail_Soft_Chromatic}
  n11 = Multiply(A=Multiply(A=@n9.Result, B=@n10.RGB)) {constB=0.25}
  n12 = TextureCoordinate() {uTiling=2}
  n13 = MaterialFunctionCall(Float2=@n12) {materialFunction=&BreakOutFloat2Components}
  n14 = Multiply(A=Constant() {r=5}) {constB=0.001}
  n15 = Multiply(A=@n14) {constB=0.7}
  n16 = Add(A=@n13.R, B=@n15)
  n17 = Add(A=@n13.G, B=@n14)
  n18 = Multiply(A=Constant2Vector() {r=-2}, B=Time())
  n19 = TextureCoordinate() {vTiling=2}
  n20 = Multiply(A=Constant2Vector() {r=-0.2}, B=Time())
  n21 = TextureSample!(UVs=Add(A=Multiply(A=@n4) {constB=0.1}, B=Add(A=@n19, B=@n20)), Tex, Apply View MipBias) {texture=&T_Inky_Smoke_Tile}
  n22 = Add(A=Multiply(A=ComponentMask(None=@n21.RGB) {g=true, r=true}) {constB=0.1}) {constB=-0.15}
  n23 = Add(A=Add(A=AppendVector(A=@n16, B=@n17), B=@n18), B=@n22)
  n24 = Constant2Vector() {g=1}
  n25 = MaterialFunctionCall(Float3=Multiply(
    A=MaterialFunctionCall(TexCoord=@n23, ClampChannel=@n24) {materialFunction=&MF_FX_TilingMethod}.Result
    B=TextureSample(UVs=@n23) {samplerType=SAMPLERTYPE_LinearColor, texture=/Game/Effects/Fort_Effects/Textures/T_CubeNinja_Trail}.RGB
   )) {materialFunction=&BreakOutFloat3Components}
  n26 = MaterialFunctionCall() {materialFunction=&LinearGradient}
  n27 = Sine(None=Multiply(A=@n26.VGradient) {constB=0.5})
  n28 = Add(A=Add(A=@n12, B=@n18), B=@n22)
  n29 = Constant2Vector() {g=1}
  n30 = MaterialFunctionCall(Float3=Multiply(
    A=MaterialFunctionCall(TexCoord=@n28, ClampChannel=@n29) {materialFunction=&MF_FX_TilingMethod}.Result
    B=TextureSample(UVs=@n28) {samplerType=SAMPLERTYPE_LinearColor, texture=&T_CubeNinja_Trail}.RGB
   )) {materialFunction=&BreakOutFloat3Components}
  n31 = MaterialFunctionCall(Float2=@n12) {materialFunction=&BreakOutFloat2Components}
  n32 = Subtract(A=@n31.R, B=@n15)
  n33 = Subtract(A=@n31.G, B=@n14)
  n34 = Add(A=Add(A=AppendVector(A=@n32, B=@n33), B=@n18), B=@n22)
  n35 = Constant2Vector() {g=1}
  n36 = MaterialFunctionCall(Float3=Multiply(
    A=MaterialFunctionCall(TexCoord=@n34, ClampChannel=@n35) {materialFunction=&MF_FX_TilingMethod}.Result
    B=TextureSample(UVs=@n34) {samplerType=SAMPLERTYPE_LinearColor, texture=&T_CubeNinja_Trail}.RGB
   )) {materialFunction=&BreakOutFloat3Components}
  n37 = MaterialFunctionCall(
   X=LinearInterpolate(A=@n25.G, B=@n25.G, Alpha=@n27)
   Y=LinearInterpolate(A=@n30.G, B=@n30.G, Alpha=@n27)
   Z=LinearInterpolate(A=@n36.G, B=@n36.G, Alpha=@n27)
  ) {materialFunction=&MakeFloat3}
  n38 = MaterialFunctionCall(Float3=Constant3Vector() {constant=<rgba 7 5 1 0>}) {materialFunction=&BreakOutFloat3Components}
  n39 = Multiply(A=Power(Base=@n37.Result, Exp=@n38.R), B=@n38.G)
  n40 = Multiply(A=@n37.Result, B=@n38.B)
  n41 = Multiply(A=Desaturation(None=Add(A=@n39, B=@n40), Fraction=Constant() {r=-1})) {constA=0.5, constB=0.7}
  n42 = VectorParameter() {parameterName=Center, defaultValue=<rgba 2 2 2 1>}
  n43 = Constant() {r=50}
  n44 = Multiply(A=Constant() {r=-0.5}, B=Time())
  n45 = Constant() {r=-0.5}
  n46 = Add(A=Multiply(A=@n5) {constB=0.5}, B=Add(A=TextureCoordinate() {uTiling=0.5, vTiling=2}, B=AppendVector(A=@n44, B=@n45)))
  n47 = Constant2Vector() {g=1}
  n48 = Power(Base=Multiply(
    A=MaterialFunctionCall(TexCoord=@n46, ClampChannel=@n47) {materialFunction=&MF_FX_TilingMethod}.Result
    B=TextureSample(UVs=@n46) {samplerType=SAMPLERTYPE_Grayscale, texture=&T_Smoke_Trail_Soft}.R
   )) {constExponent=6}
  n49 = Multiply(
   A=LinearInterpolate(
    A=CurveAtlasRowParameter(CurveTime=@n1) {atlas=&/CurveAtlas_LifeDrain_Vampire, curve=&/CC_VampireDrain, parameterName=Param_1}
    B=CurveAtlasRowParameter(CurveTime=@n1) {atlas=&CurveAtlas_LifeDrain_Vampire, curve=&CC_VampireDrain, parameterName=Param_2}
    Alpha=DynamicParameter() {defaultValue=<rgba 1 1.1 1 1>, paramNames [Dissolve, DissolveEdge, Param3, Curve Color Mix]}.Curve Color Mix
   )
   B=LinearInterpolate(A=Add(A=@n11, B=@n41), B=Multiply(A=@n42.RGB, B=@n43), Alpha=@n48)
  )
  n50 = Constant() {r=-0.35}
  n51 = ParticleColor()
  MaterialFunctionCall(Emissive=Multiply(A=Desaturation(None=@n49, Fraction=@n50), B=@n51.RGB)) {materialFunction=&MF_FX_EmissiveAdjustments}
  n52 = Constant() {r=1}
  n53 = Add(A=Multiply(A=@n9.Result, B=@n10.G), B=Multiply(A=Desaturation(None=@n37.Result, Fraction=@n52)) {constB=0.8})
  n54 = TextureCoordinate() {vTiling=2}
  n55 = Multiply(A=TextureSample(UVs=Panner(Coordinate=@n54)) {texture=&T_Inky_Smoke_Tile}.R) {constB=0.5}
  n56 = Constant() {r=3}
  n57 = Multiply(A=Saturate(None=Add(A=@n26.UGradient, B=@n55)), B=@n56) {constB=3}
  n58 = LinearInterpolate(A=@n56, Alpha=DynamicParameter() {defaultValue=<rgba 1 1.1 1 1>, paramNames [Dissolve, DissolveEdge, Param3, Curve Color Mix]}.Dissolve) {constA=3, constB=-1}
  n59 = Multiply(A=Multiply(A=LinearInterpolate(A=@n53, B=@n43, Alpha=@n48) {constB=200}, B=@n51.A), B=Saturate(None=Subtract(A=@n57, B=@n58)))
  n60 = DepthFade(Opacity=@n59) {fadeDistanceDefault=50}
  n61 = OneMinus(None=MaterialFunctionCall() {materialFunction=&LinearGradient}.UGradient)
  Multiply(A=ShadingPathSwitch(Default=@n60, Deferred=@n60, Forward=@n59, Mobile=@n59), B=OneMinus(None=Power(Base=@n61) {constExponent=5}))
  ScalarParameter() {parameterName=Curve Color Mix}
  n62 = LinearInterpolate(Alpha=@n60) {constA=1, constB=1.05}
  n63 = Constant() {r=1}
  n64 = QualitySwitch(Default=@n62, Low=@n63, High=@n62, Medium=@n63, Epic=@n62)
  ShadingPathSwitch(Default=@n64, Deferred=@n64, Forward=@n63, Mobile=@n63)
 ]
}
